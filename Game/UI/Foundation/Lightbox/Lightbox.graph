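# Lightbox is a reusable CommonUserWidget backdrop that dims the screen behind content placed in its safe-zone named slot, with an invisible button beneath to catch clicks.
# AddContent and RemoveContent fill or clear the slot.
# Intro and Outro play half-second fades on the Dimmer border, firing IntroEnded and OutroEnded dispatchers when finished.
# Construct tints the radial-gradient darkening border from BackgroundColor and DimAmount, and toggles a platform-filtered window title bar.
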WidgetBlueprint /Game/UI/Foundation/Lightbox/Lightbox
animations [
 FadeIn {
  binding_count=1
  display_label=""
  length_seconds=0.5
  object=&self:FadeIn
  sequence {bindings [Dimmer {
     id=#1
     possessed_class=Border
     tracks [
      {
       type=MovieSceneColorTrack
       display_name=Brush Color
       sections [{
         type=MovieSceneColorSection
         end_frame=10
         start_frame=0
         channels [
          Color.R {type=MovieSceneScriptingFloatChannel}
          Color.G {type=MovieSceneScriptingFloatChannel}
          Color.B {type=MovieSceneScriptingFloatChannel}
          Color.A {type=MovieSceneScriptingFloatChannel, keys [[0,0],[10,1]]}
         ]
        }]
      }
      {
       type=MovieSceneColorTrack
       display_name=Content Color and Opacity
       sections [{
         type=MovieSceneColorSection
         end_frame=10
         start_frame=0
         channels [
          Color.R {type=MovieSceneScriptingFloatChannel, const [1,1]}
          Color.G {type=MovieSceneScriptingFloatChannel, const [1,1]}
          Color.B {type=MovieSceneScriptingFloatChannel, const [1,1]}
          Color.A {type=MovieSceneScriptingFloatChannel, keys [[0,0],[10,1]]}
         ]
        }]
      }
     ]
    }]}
 }
 FadeOut {
  binding_count=1
  display_label=""
  length_seconds=0.5
  object=&self:FadeOut
  sequence {bindings [Dimmer {
     id=#1
     possessed_class=Border
     tracks [
      {
       type=MovieSceneColorTrack
       display_name=Brush Color
       sections [{
         type=MovieSceneColorSection
         end_frame=10
         start_frame=0
         channels [
          Color.R {type=MovieSceneScriptingFloatChannel}
          Color.G {type=MovieSceneScriptingFloatChannel}
          Color.B {type=MovieSceneScriptingFloatChannel}
          Color.A {type=MovieSceneScriptingFloatChannel, keys [[0,1],[10,0]]}
         ]
        }]
      }
      {
       type=MovieSceneColorTrack
       display_name=Content Color and Opacity
       sections [{
         type=MovieSceneColorSection
         end_frame=10
         start_frame=0
         channels [
          Color.R {type=MovieSceneScriptingFloatChannel, const [1,1]}
          Color.G {type=MovieSceneScriptingFloatChannel, const [1,1]}
          Color.B {type=MovieSceneScriptingFloatChannel, const [1,1]}
          Color.A {type=MovieSceneScriptingFloatChannel, keys [[0,1],[10,0]]}
         ]
        }]
      }
     ]
    }]}
 }
]
graphs [
 AddContent {nodes [
   n1 = AddContent!() -> (then, Content:Common User Widget Object Reference) {is=title K2Node_FunctionEntry}
   > AddChild(self=Get Content() -> (Content:Named Slot Object Reference), Content=@n1)
  ]}
 EventGraph {nodes [
   n1 = Event Construct()
   n2 = SetVisibility(
    execute=[@n1.then, @n5.then]
    self=Get WindowsOverlay() -> (WindowsOverlay:Overlay Object Reference)
    InVisibility=Select(Option 0=Collapsed, Option 1=Visible, Index=Get ShowWindowControlButtons() -> (ShowWindowControlButtons:bool))
   )
   n3 = Break Linear Color!(LinearColor:"Linear Color Structure (by ref)"=Get BackgroundColor() -> (BackgroundColor:Linear Color Structure)) -> (R:float=~, G:float=~, B:float=~, A:float=~) {is=title K2Node_BreakStruct}
   SetBrushColor(
    execute=[@n2.then, @n5.else]
    self=Get Border_Darken() -> (Border_Darken:Border Object Reference)
    InBrushColor=MakeColor!(R:float=@n3, G:float=@n3.G, B:float=@n3.B, A:float=Get DimAmount() -> (DimAmount:double)) -> (ReturnValue:Linear Color Structure)
   )
   n4 = Event PreConstruct()
   > n5 = Branch(Condition=@n4.IsDesignTime)
   "Animation Finished (FadeIn)"!() -> (OutputDelegate:Delegate, then:Exec) {is=title K2Node_WidgetAnimationEvent}
   > Call Intro Ended!(execute:Exec, self:Lightbox Object Reference) -> (then:Exec) {is=title K2Node_CallDelegate}
   "Animation Finished (FadeOut)"!() -> (OutputDelegate:Delegate, then:Exec) {is=title K2Node_WidgetAnimationEvent}
   > Call Outro Ended!(execute:Exec, self:Lightbox Object Reference) -> (then:Exec) {is=title K2Node_CallDelegate}
   Event Destruct()
  ]}
 Intro {nodes [
   Intro() {is=title K2Node_FunctionEntry}
   > PlayAnimation(InAnimation=Get FadeIn() -> (FadeIn:Widget Animation Object Reference))
  ]}
 IntroEnded {nodes [IntroEnded() {is=title K2Node_FunctionEntry}]}
 Outro {nodes [
   Outro() {is=title K2Node_FunctionEntry}
   > PlayAnimation(InAnimation=Get FadeOut() -> (FadeOut:Widget Animation Object Reference))
  ]}
 OutroEnded {nodes [OutroEnded() {is=title K2Node_FunctionEntry}]}
 RemoveContent {nodes [
   RemoveContent() {is=title K2Node_FunctionEntry}
   > Branch(Condition=HasAnyChildren!(self:Panel Widget Object Reference=Get Content()) -> (ReturnValue:bool))
   > ClearChildren(self=Get Content())
  ]}
]
variables [
 BackgroundColor {type=LinearColor, default=<rgba 0.0052 0.0052 0.0056 1>}
 DimAmount {type=number, default=0.5}
 IntroEnded {}
 OutroEnded {}
 ShowWindowControlButtons {type=boolean, default=true}
]
widgets {
 info {inheritedWidgetCount=0, parentClass=&CommonUserWidget, rootWidgetClass=&Border, widgetCount=11}
 widgets [
  Dimmer {
   widgetClassPath=&Border
   bIsVariable=true
   widget=Dimmer
   properties {
    visibility=SelfHitTestInvisible
    background {
     bIsDynamicallyLoaded=false
     drawAs=NoDrawType
     imageType=NoImage
     mirroring=NoMirror
     resourceName=None
     resourceObject=null
     tiling=NoTile
     imageSize=<32 32>
     margin=<ltrb 0 0 0 0>
     outlineSettings {
      bUseBrushTransparency=false
      roundingType=HalfHeightRadius
      width=0
      color {colorUseRule=UseColor_Specified, specifiedColor=<rgba 0 0 0 0>}
      cornerRadii=<0 0 0 0>
     }
     tintColor {colorUseRule=UseColor_Specified, specifiedColor=<rgba 1 1 1 1>}
     uVRegion {bIsValid=false, max=<0 0>, min=<0 0>}
    }
    padding=<ltrb 0 0 0 0>
   }
  }
  Overlay_0 {widgetClassPath=&Overlay, parent=Dimmer, widget=Overlay_0, slot {class=BorderSlot, properties {padding=<ltrb 0 0 0 0>}}}
  ClickCatcher {
   widgetClassPath=&Button
   parent=Overlay_0
   widget=ClickCatcher
   properties {
    backgroundColor=<rgba 1 1 1 0>
    colorAndOpacity=<rgba 1 1 1 0>
    widgetStyle {
     clickedSlateSound {resourceObject=null}
     disabled {
      bIsDynamicallyLoaded=false
      drawAs=NoDrawType
      imageType=NoImage
      mirroring=NoMirror
      resourceName=None
      resourceObject=null
      tiling=NoTile
      imageSize=<0 0>
      margin=<ltrb 0 0 0 0>
      outlineSettings {
       bUseBrushTransparency=false
       roundingType=HalfHeightRadius
       width=0
       color {colorUseRule=UseColor_Specified, specifiedColor=<rgba 0 0 0 0>}
       cornerRadii=<0 0 0 0>
      }
      tintColor {colorUseRule=UseColor_Specified, specifiedColor=<rgba 1 1 1 0>}
      uVRegion {bIsValid=false, max=<0 0>, min=<0 0>}
     }
     disabledForeground {colorUseRule=UseColor_Specified, specifiedColor=<rgba 0.5271 0.5271 0.5271 1>}
     hovered {
      bIsDynamicallyLoaded=false
      drawAs=NoDrawType
      imageType=NoImage
      mirroring=NoMirror
      resourceName=None
      resourceObject=null
      tiling=NoTile
      imageSize=<32 32>
      margin=<ltrb 0 0 0 0>
      outlineSettings {
       bUseBrushTransparency=true
       roundingType=FixedRadius
       width=1
       color {colorUseRule=UseColor_Specified, specifiedColor=<rgba 0.7243 0.7243 0.7243 1>}
       cornerRadii=<4 4 4 4>
      }
      tintColor {colorUseRule=UseColor_Specified, specifiedColor=<rgba 1 1 1 0>}
      uVRegion {bIsValid=false, max=<0 0>, min=<0 0>}
     }
     hoveredForeground {colorUseRule=UseColor_Specified, specifiedColor=<rgba 0 1 1 1>}
     hoveredSlateSound {resourceObject=null}
     normal {
      bIsDynamicallyLoaded=false
      drawAs=NoDrawType
      imageType=NoImage
      mirroring=NoMirror
      resourceName=None
      resourceObject=null
      tiling=NoTile
      imageSize=<32 32>
      margin=<ltrb 0 0 0 0>
      outlineSettings {
       bUseBrushTransparency=true
       roundingType=FixedRadius
       width=1
       color {colorUseRule=UseColor_Specified, specifiedColor=<rgba 0.6951 0.6951 0.6951 1>}
       cornerRadii=<4 4 4 4>
      }
      tintColor {colorUseRule=UseColor_Specified, specifiedColor=<rgba 1 1 1 0>}
      uVRegion {bIsValid=false, max=<0 0>, min=<0 0>}
     }
     normalForeground {colorUseRule=UseColor_Specified, specifiedColor=<rgba 0 1 1 1>}
     normalPadding=<ltrb 0 0 0 0>
     pressed {
      bIsDynamicallyLoaded=false
      drawAs=NoDrawType
      imageType=NoImage
      mirroring=NoMirror
      resourceName=None
      resourceObject=null
      tiling=NoTile
      imageSize=<32 32>
      margin=<ltrb 0 0 0 0>
      outlineSettings {
       bUseBrushTransparency=true
       roundingType=FixedRadius
       width=1
       color {colorUseRule=UseColor_Specified, specifiedColor=<rgba 0.7243 0.7243 0.7243 1>}
       cornerRadii=<4 4 4 4>
      }
      tintColor {colorUseRule=UseColor_Specified, specifiedColor=<rgba 1 1 1 0>}
      uVRegion {bIsValid=false, max=<0 0>, min=<0 0>}
     }
     pressedForeground {colorUseRule=UseColor_Specified, specifiedColor=<rgba 0 1 1 1>}
     pressedPadding=<ltrb 0 0 0 0>
     pressedSlateSound {resourceObject=null}
    }
   }
   slot {class=OverlaySlot, properties {horizontalAlignment=HAlign_Fill, verticalAlignment=VAlign_Fill}}
  }
  Border_Darken {
   widgetClassPath=&Border
   bIsVariable=true
   parent=Overlay_0
   widget=Border_Darken
   properties {
    visibility=SelfHitTestInvisible
    background {
     bIsDynamicallyLoaded=false
     drawAs=Image
     imageType=NoImage
     mirroring=NoMirror
     resourceName=None
     resourceObject=/Game/Athena/UI/Frontend/Art/M_UI_RadialGradientMaster_Inst
     tiling=NoTile
     imageSize=<32 32>
     margin=<ltrb 0 0 0 0>
     outlineSettings {
      bUseBrushTransparency=false
      roundingType=HalfHeightRadius
      width=0
      color {colorUseRule=UseColor_Specified, specifiedColor=<rgba 0 0 0 0>}
      cornerRadii=<0 0 0 0>
     }
     tintColor {colorUseRule=UseColor_Specified, specifiedColor=<rgba 1 1 1 1>}
     uVRegion {bIsValid=false, max=<0 0>, min=<0 0>}
    }
    brushColor=<rgba 0.0869 0.0869 0.094 0.5>
    padding=<ltrb 0 0 0 0>
   }
   slot {class=OverlaySlot, properties {horizontalAlignment=HAlign_Fill, verticalAlignment=VAlign_Fill}}
  }
  SafeZone_0 {
   widgetClassPath=&SafeZone
   bIsVariable=true
   parent=Overlay_0
   widget=SafeZone_0
   slot {class=OverlaySlot, properties {horizontalAlignment=HAlign_Fill, verticalAlignment=VAlign_Fill}}
  }
  Content {
   widgetClassPath=&NamedSlot
   bIsVariable=true
   parent=SafeZone_0
   widget=Content
   properties {bExposeOnInstanceOnly=true}
   slot {class=SafeZoneSlot}
  }
  WindowsOverlay {
   widgetClassPath=&Overlay
   bIsVariable=true
   parent=Overlay_0
   widget=WindowsOverlay
   properties {visibility=Visible}
   slot {class=OverlaySlot, properties {horizontalAlignment=HAlign_Fill}}
  }
  WIndowButtons {
   widgetClassPath=&CommonVisibilityWidgetLegacy
   bIsVariable=true
   parent=WindowsOverlay
   widget=WIndowButtons
   properties {
    bShowForAndroid=false
    bShowForErebus=false
    bShowForIOS=false
    bShowForMac=false
    bShowForPS4=false
    bShowForSwitch2=false
    bShowForXBox=false
    padding=<ltrb 0 0 0 0>
   }
   slot {class=OverlaySlot, properties {horizontalAlignment=HAlign_Fill, verticalAlignment=VAlign_Fill}}
  }
  SizeBox_0 {
   widgetClassPath=&SizeBox
   parent=WIndowButtons
   widget=SizeBox_0
   properties {bOverride_HeightOverride=true, heightOverride=64}
   slot {class=BorderSlot, properties {padding=<ltrb 0 0 0 0>}}
  }
  Overlay_1 {widgetClassPath=&Overlay, parent=SizeBox_0, widget=Overlay_1, slot {class=SizeBoxSlot}}
  WindowTitleBar {
   widgetClassPath=/Game/UI/Foundation/Shell/windowchromebuttons
   bIsVariable=true
   parent=Overlay_1
   widget=WindowTitleBar
   slot {class=OverlaySlot, properties {horizontalAlignment=HAlign_Fill, verticalAlignment=VAlign_Fill}}
  }
 ]
}
